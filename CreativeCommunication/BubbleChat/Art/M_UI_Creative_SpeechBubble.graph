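# Translucent UI material that draws a Creative chat speech bubble with a positionable tail, tinted by an RGBA parameter.
# The body combines an offset 9-slice box SDF, less its antialiased outline, with a narrow center-band mask.
# The tail, blended between a channel-selected texture and a skewed triangle SDF, can be rotated, scaled, offset or flipped, then is added to the body.
# A "Uses Blur" switch mixes in the Slate post buffer and drops the alpha tint.
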Material /CreativeCommunication/BubbleChat/Art/M_UI_Creative_SpeechBubble
graph {
 expressions {
  MaterialExpressionAbs_0 {class=MaterialExpressionAbs, inputs [None {from=MaterialExpressionAdd_3, output=""}], properties {}}
  MaterialExpressionAdd_0 {
   class=MaterialExpressionAdd
   inputs [
    A {from=MaterialExpressionSaturate_4}
    B {from=MaterialExpressionLinearInterpolate_0}
   ]
  }
  MaterialExpressionAdd_1 {
   class=MaterialExpressionAdd
   inputs [
    A {from=MaterialExpressionMaterialFunctionCall_3, output=9-Slice UV}
    B {from=MaterialExpressionAppendVector_0}
   ]
  }
  MaterialExpressionAdd_3 {
   class=MaterialExpressionAdd
   inputs [
    A {from=MaterialExpressionAdd_1}
    B {from=""}
   ]
   properties {constB=-0.5}
  }
  MaterialExpressionAdd_4 {
   class=MaterialExpressionAdd
   inputs [
    A {from=MaterialExpressionMaterialFunctionCall_6, output=UVs}
    B {from=MaterialExpressionAppendVector_3}
   ]
  }
  MaterialExpressionAdd_5 {
   class=MaterialExpressionAdd
   inputs [
    A {from=MaterialExpressionAbs_0}
    B {from=""}
   ]
   properties {constB=-0.05}
  }
  MaterialExpressionAdd_6 {
   class=MaterialExpressionAdd
   inputs [
    A {from=MaterialExpressionSaturate_3}
    B {from=MaterialExpressionSaturate_1}
   ]
  }
  MaterialExpressionAppendVector_0 {
   class=MaterialExpressionAppendVector
   inputs [
    A {from=MaterialExpressionScalarParameter_4, output=""}
    B {from=MaterialExpressionScalarParameter_0, output=""}
   ]
  }
  MaterialExpressionAppendVector_1 {
   class=MaterialExpressionAppendVector
   inputs [
    A {from=MaterialExpressionScalarParameter_11, output=""}
    B {from=MaterialExpressionScalarParameter_12, output=""}
   ]
  }
  MaterialExpressionAppendVector_3 {
   class=MaterialExpressionAppendVector
   inputs [
    A {from=MaterialExpressionScalarParameter_1, output=""}
    B {from=MaterialExpressionScalarParameter_5, output=""}
   ]
  }
  MaterialExpressionAppendVector_4 {
   class=MaterialExpressionAppendVector
   inputs [
    A {from=MaterialExpressionOneMinus_0, output=""}
    B {from=MaterialExpressionMaterialFunctionCall_5, output=G}
   ]
  }
  MaterialExpressionConstant_0 {class=MaterialExpressionConstant, properties {r=0.2}}
  MaterialExpressionConstant_2 {class=MaterialExpressionConstant, properties {r=3}}
  MaterialExpressionConstant_4 {class=MaterialExpressionConstant, properties {r=0.2}}
  MaterialExpressionConstant_5 {class=MaterialExpressionConstant, properties {r=1}}
  MaterialExpressionConstant_8 {class=MaterialExpressionConstant, properties {r=0.5}}
  MaterialExpressionLinearInterpolate_0 {
   class=MaterialExpressionLinearInterpolate
   inputs [
    A {from=MaterialExpressionMaterialFunctionCall_2, output=Outer}
    B {from=MaterialExpressionMaterialFunctionCall_11, output=Inner}
    Alpha {from=MaterialExpressionScalarParameter_9}
   ]
  }
  MaterialExpressionLinearInterpolate_2 {
   class=MaterialExpressionLinearInterpolate
   inputs [
    A {from=MaterialExpressionMaterialFunctionCall_1, output=LinearRGB}
    B {from=MaterialExpressionVectorParameter_0, output=RGB}
    Alpha {from=MaterialExpressionVectorParameter_0, output=RGB}
   ]
   properties {constAlpha=0.975}
  }
  MaterialExpressionMaterialFunctionCall_0 {class=MaterialExpressionMaterialFunctionCall, inputs [], properties {materialFunction=&GetUserInterfaceUV}}
  MaterialExpressionMaterialFunctionCall_1 {
   class=MaterialExpressionMaterialFunctionCall
   inputs ["UVs (Optional)" {from=""}]
   properties {materialFunction=/Engine/Functions/UserInterface/GetSlatePost0}
  }
  MaterialExpressionMaterialFunctionCall_10 {
   class=MaterialExpressionMaterialFunctionCall
   inputs [
    DistanceField {from=MaterialExpressionMaterialFunctionCall_4, output=SDFOutline}
    Threshold {from=""}
    Smoothing {from=""}
   ]
   properties {materialFunction=&MF_UI_AntiAliasing}
  }
  MaterialExpressionMaterialFunctionCall_11 {
   class=MaterialExpressionMaterialFunctionCall
   inputs [
    DistanceField {from=MaterialExpressionMaterialFunctionCall_14, output=SDF}
    Threshold {from=MaterialExpressionConstant_4}
    Smoothing {from=""}
   ]
   properties {materialFunction=&MF_UI_AntiAliasing}
  }
  MaterialExpressionMaterialFunctionCall_14 {
   class=MaterialExpressionMaterialFunctionCall
   inputs [
    UVs {from=MaterialExpressionMaterialFunctionCall_16, output=Skewed UVs}
    Size {from=MaterialExpressionScalarParameter_10}
    Sides {from=MaterialExpressionConstant_2}
    InnerAngle {from=""}
    OutlineThickness {from=""}
    OutlineAlignment {from=""}
    GlowMin {from=""}
    GlowMax {from=""}
   ]
   properties {materialFunction=/FortUILibrary/MaterialFunctions/Shapes/MF_UI_RegularPolygon}
  }
  MaterialExpressionMaterialFunctionCall_16 {
   class=MaterialExpressionMaterialFunctionCall
   inputs [
    UV {from=MaterialExpressionAdd_4}
    Skew Angle {from=MaterialExpressionScalarParameter_13}
   ]
   properties {materialFunction=/FortUILibrary/MaterialFunctions/Basics/MF_UI_Skew}
  }
  MaterialExpressionMaterialFunctionCall_18 {
   class=MaterialExpressionMaterialFunctionCall
   inputs [Multiplier {from=MaterialExpressionScalarParameter_8}]
   properties {materialFunction=&Pi}
  }
  MaterialExpressionMaterialFunctionCall_2 {
   class=MaterialExpressionMaterialFunctionCall
   inputs [
    DistanceField {from=MaterialExpressionSwitch_0}
    Threshold {from=MaterialExpressionConstant_0}
    Smoothing {from=""}
   ]
   properties {materialFunction=&MF_UI_AntiAliasing}
  }
  MaterialExpressionMaterialFunctionCall_3 {class=MaterialExpressionMaterialFunctionCall, inputs [], properties {materialFunction=&GetUserInterfaceUV}}
  MaterialExpressionMaterialFunctionCall_4 {
   class=MaterialExpressionMaterialFunctionCall
   inputs [
    UVs {from=MaterialExpressionAdd_1}
    Size {from=MaterialExpressionScalarParameter_3}
    CornerRoundness {from=MaterialExpressionScalarParameter_3}
    OutlineThickness {from=MaterialExpressionScalarParameter_7}
    OutlineAlignment {from=""}
    GlowMin {from=""}
    GlowMax {from=""}
    KeepAspectRatio {from=MaterialExpressionStaticBoolParameter_0}
    PixelSize {from=""}
   ]
   properties {materialFunction=&MF_UI_Box}
  }
  MaterialExpressionMaterialFunctionCall_46 {
   class=MaterialExpressionMaterialFunctionCall
   inputs [
    UV {from=MaterialExpressionMaterialFunctionCall_0, output=Normalized UV}
    TextureSize {from=""}
    ScalePivotPoint {from=MaterialExpressionAppendVector_1}
    ContainerSize {from=""}
   ]
   properties {materialFunction=&MF_UI_UVScaling}
  }
  MaterialExpressionMaterialFunctionCall_5 {
   class=MaterialExpressionMaterialFunctionCall
   inputs [Float2 {from=MaterialExpressionRotator_0}]
   properties {materialFunction=&BreakOutFloat2Components}
  }
  MaterialExpressionMaterialFunctionCall_6 {
   class=MaterialExpressionMaterialFunctionCall
   inputs [
    UVs {from=MaterialExpressionStaticSwitchParameter_0}
    Texture Scale {from=MaterialExpressionMultiply_0}
   ]
   properties {materialFunction=&ScaleUVsByCenter}
  }
  MaterialExpressionMaterialFunctionCall_7 {
   class=MaterialExpressionMaterialFunctionCall
   inputs [
    DistanceField {from=MaterialExpressionMaterialFunctionCall_4, output=SDF}
    Threshold {from=""}
    Smoothing {from=""}
   ]
   properties {materialFunction=&MF_UI_AntiAliasing}
  }
  MaterialExpressionMaterialFunctionCall_8 {
   class=MaterialExpressionMaterialFunctionCall
   inputs [Float2 {from=MaterialExpressionOneMinus_1}]
   properties {materialFunction=&BreakOutFloat2Components}
  }
  MaterialExpressionMultiply_0 {
   class=MaterialExpressionMultiply
   inputs [
    A {from=MaterialExpressionConstant_8}
    B {from=MaterialExpressionScalarParameter_2}
   ]
  }
  MaterialExpressionMultiply_5 {
   class=MaterialExpressionMultiply
   inputs [
    A {from=MaterialExpressionVectorParameter_0, output=A}
    B {from=MaterialExpressionSaturate_0}
   ]
  }
  MaterialExpressionMultiply_6 {
   class=MaterialExpressionMultiply
   inputs [
    A {from=MaterialExpressionAdd_5}
    B {from=""}
   ]
   properties {constB=256}
  }
  MaterialExpressionMultiply_7 {
   class=MaterialExpressionMultiply
   inputs [
    A {from=MaterialExpressionMaterialFunctionCall_8, output=R}
    B {from=MaterialExpressionMaterialFunctionCall_8, output=R}
   ]
  }
  MaterialExpressionOneMinus_0 {class=MaterialExpressionOneMinus, inputs [None {from=MaterialExpressionMaterialFunctionCall_5, output=R}]}
  MaterialExpressionOneMinus_1 {class=MaterialExpressionOneMinus, inputs [None {from=MaterialExpressionSaturate_2}]}
  MaterialExpressionReroute_0 {class=MaterialExpressionReroute, inputs [None {from=MaterialExpressionReroute_1}]}
  MaterialExpressionReroute_1 {class=MaterialExpressionReroute, inputs [None {from=MaterialExpressionVectorParameter_0, output=RGB}]}
  MaterialExpressionReroute_2 {class=MaterialExpressionReroute, inputs [None {from=MaterialExpressionMultiply_5}]}
  MaterialExpressionReroute_3 {class=MaterialExpressionReroute, inputs [None {from=MaterialExpressionSaturate_0}]}
  MaterialExpressionRotator_0 {
   class=MaterialExpressionRotator
   inputs [
    Coordinate {from=MaterialExpressionMaterialFunctionCall_46, output=Scaled to Fit}
    Time {from=MaterialExpressionMaterialFunctionCall_18, output=Result}
   ]
   properties {}
  }
  MaterialExpressionSaturate_0 {class=MaterialExpressionSaturate, inputs [None {from=MaterialExpressionAdd_0}]}
  MaterialExpressionSaturate_1 {class=MaterialExpressionSaturate, inputs [None {from=MaterialExpressionSubtract_0}]}
  MaterialExpressionSaturate_2 {class=MaterialExpressionSaturate, inputs [None {from=MaterialExpressionMultiply_6}]}
  MaterialExpressionSaturate_3 {class=MaterialExpressionSaturate, inputs [None {from=MaterialExpressionMultiply_7}]}
  MaterialExpressionSaturate_4 {class=MaterialExpressionSaturate, inputs [None {from=MaterialExpressionAdd_6}]}
  MaterialExpressionScalarParameter_0 {class=MaterialExpressionScalarParameter, properties {parameterName=Bubble_Offset_Y}}
  MaterialExpressionScalarParameter_1 {class=MaterialExpressionScalarParameter, properties {parameterName=Tail_Position_X}}
  MaterialExpressionScalarParameter_10 {class=MaterialExpressionScalarParameter, properties {defaultValue=5, parameterName=TailSharpness}}
  MaterialExpressionScalarParameter_11 {class=MaterialExpressionScalarParameter, properties {defaultValue=0.5, parameterName=ScalePivot_X}}
  MaterialExpressionScalarParameter_12 {class=MaterialExpressionScalarParameter, properties {defaultValue=1, parameterName=ScalePivot_Y}}
  MaterialExpressionScalarParameter_13 {class=MaterialExpressionScalarParameter, properties {parameterName=TailSkewedAngle}}
  MaterialExpressionScalarParameter_2 {class=MaterialExpressionScalarParameter, properties {defaultValue=1, parameterName=TailScale}}
  MaterialExpressionScalarParameter_3 {class=MaterialExpressionScalarParameter, properties {defaultValue=1, parameterName=CircleSize}}
  MaterialExpressionScalarParameter_4 {class=MaterialExpressionScalarParameter, properties {parameterName=Bubble_Offset_X}}
  MaterialExpressionScalarParameter_5 {class=MaterialExpressionScalarParameter, properties {defaultValue=-0.63, parameterName=Tail_Position_Y}}
  MaterialExpressionScalarParameter_6 {class=MaterialExpressionScalarParameter, properties {parameterName=Channel}}
  MaterialExpressionScalarParameter_7 {class=MaterialExpressionScalarParameter, properties {defaultValue=0.5, parameterName=OutlineThickness}}
  MaterialExpressionScalarParameter_8 {class=MaterialExpressionScalarParameter, properties {desc="0=Tail bottom, 4=Tail top, 2=tail right, 6=tail left", parameterName=TailAngle}}
  MaterialExpressionScalarParameter_9 {class=MaterialExpressionScalarParameter, properties {parameterName=isTailSDF?}}
  MaterialExpressionStaticBoolParameter_0 {class=MaterialExpressionStaticBoolParameter, inputs [], properties {defaultValue=true, parameterName=KeepAspectRatio}}
  MaterialExpressionStaticSwitchParameter_0 {
   class=MaterialExpressionStaticSwitchParameter
   inputs [
    True {from=MaterialExpressionAppendVector_4, output=""}
    False {from=MaterialExpressionRotator_0, output=""}
   ]
   properties {parameterName=FlipTailHorizontally}
  }
  MaterialExpressionStaticSwitchParameter_1 {
   class=MaterialExpressionStaticSwitchParameter
   inputs [
    True {from=MaterialExpressionLinearInterpolate_2, output=""}
    False {from=MaterialExpressionReroute_0, output=""}
   ]
   properties {parameterName=Uses Blur}
  }
  MaterialExpressionStaticSwitchParameter_2 {
   class=MaterialExpressionStaticSwitchParameter
   inputs [
    True {from=MaterialExpressionReroute_3, output=""}
    False {from=MaterialExpressionReroute_2, output=""}
   ]
   properties {parameterName=Uses Blur}
  }
  MaterialExpressionSubtract_0 {
   class=MaterialExpressionSubtract
   inputs [
    A {from=MaterialExpressionMaterialFunctionCall_7, output=Inner}
    B {from=MaterialExpressionMaterialFunctionCall_10, output=Inner}
   ]
  }
  MaterialExpressionSwitch_0 {
   class=MaterialExpressionSwitch
   inputs [
    SwitchValue {from=MaterialExpressionScalarParameter_6, output=""}
    Default {from="", output=""}
    None {from=MaterialExpressionTextureSample_0, output=R}
    None {from=MaterialExpressionTextureSample_0, output=R}
    None {from=MaterialExpressionTextureSample_0, output=R}
   ]
   properties {inputs [
     {
      inputName=None
      input {
       expression=&self:MaterialExpressionTextureSample_0
       inputName=None
       mask=1
       maskA=0
       maskB=0
       maskG=0
       maskR=1
       outputIndex=1
      }
     }
     {
      inputName=None
      input {
       expression=&self:MaterialExpressionTextureSample_0
       inputName=None
       mask=1
       maskA=0
       maskB=0
       maskG=1
       maskR=0
       outputIndex=2
      }
     }
     {
      inputName=None
      input {
       expression=&self:MaterialExpressionTextureSample_0
       inputName=None
       mask=1
       maskA=0
       maskB=1
       maskG=0
       maskR=0
       outputIndex=3
      }
     }
    ]}
  }
  MaterialExpressionTextureSample_0 {
   class=MaterialExpressionTextureSample
   inputs [
    UVs {from=MaterialExpressionAdd_4}
    Tex {from=""}
    Apply View MipBias {from=""}
   ]
   properties {texture=&/T_UI_Creative_Speechtail}
  }
  MaterialExpressionVectorParameter_0 {class=MaterialExpressionVectorParameter, properties {parameterName=RGBA, defaultValue=<rgba 1 1 1 1>}}
 }
 properties {bAutomaticallySetUsageInEditor=false, blendMode=BLEND_Translucent, materialDomain=MD_UI, usedShadingModels=Default Lit}
}
